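# Sound cue for the third-person active loop of the Stimpulse Adrenaline item, classed as a consumable sound.
# A mixer layers an immediate random one-shot burst over a briefly delayed, randomly chosen looping wave, with the burst weighted louder.
# The loop is pitched down with slight variation and fades in through an enveloper volume curve.
# A concurrency override allows one instance and blocks new ones, and attenuation uses shared first-person weapon foley settings.
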SoundCue /ImpulseGrenade/Gameplay/StimpulseGrenade/Gameplay/Items/Adrenaline/Audio/SC_Item_FP_STAdrenaline_Active_Loop_3P
root=Mixer_0
nodes [
 {id=Delay_1, children [Modulator_0], properties {delayMax=0.1, delayMin=0.1}}
 {
  id=Enveloper_0
  children [Random_0]
  properties {pitchCurve {editorCurveData {keys [{time=0}]}}, volumeCurve {editorCurveData {keys [<at 0 0.0017>,<at 0.5 1>]}}}
 }
 {id=Mixer_0, children [Delay_1,Random_1], properties {inputVolume [1.5,2.6645]}}
 {id=Modulator_0, children [Enveloper_0], properties {pitchMax=0.75, pitchMin=0.7}}
 {id=Random_0, children [WavePlayer_2,WavePlayer_3], properties {weights [1,1]}}
 {id=Random_1, children [WavePlayer_4,WavePlayer_5,WavePlayer_6], properties {weights [1,1,1]}}
 {
  id=WavePlayer_2
  properties {
   bLooping=true
   soundWaveAssetPtr=/ImpulseGrenade/Gameplay/StimpulseGrenade/Gameplay/Items/Adrenaline/Audio/Waves/Stim_STAdrenaline_Active_Loop_01
  }
 }
 {
  id=WavePlayer_3
  properties {
   bLooping=true
   soundWaveAssetPtr=/ImpulseGrenade/Gameplay/StimpulseGrenade/Gameplay/Items/Adrenaline/Audio/Waves/Stim_STAdrenaline_Active_Loop_02
  }
 }
 {
  id=WavePlayer_4
  properties {soundWaveAssetPtr=/ImpulseGrenade/Gameplay/StimpulseGrenade/Gameplay/Items/Adrenaline/Audio/Waves/FeralCorge_Adrenaline_Active_Burst_Extended_04}
 }
 {
  id=WavePlayer_5
  properties {soundWaveAssetPtr=/ImpulseGrenade/Gameplay/StimpulseGrenade/Gameplay/Items/Adrenaline/Audio/Waves/FeralCorge_Adrenaline_Active_Burst_Extended_05}
 }
 {
  id=WavePlayer_6
  properties {soundWaveAssetPtr=/ImpulseGrenade/Gameplay/StimpulseGrenade/Gameplay/Items/Adrenaline/Audio/Waves/FeralCorge_Adrenaline_Active_Burst_Extended_06}
 }
]
properties {
 attenuationSettings=/FortMetaSoundWeapon/Items/Shared/Settings/FPS/AS_Weapon_FP_Foley
 bOverrideConcurrency=true
 maxDistance=2250
 soundClassObject=&Fort_Consumables
 volumeMultiplier=1
 concurrencyOverrides {maxCount=1, resolutionRule=PreventNew, retriggerTime=0.4}
}
